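# Niagara system for the idle state of the Skyfire bounce pad, showing a persistent green glow.
# A single emitter, LightCone_mesh, burst-spawns one particle that is not killed when its lifetime elapses and drives mesh, light and sprite renderers together.
# Separate sprite color, raised position and camera offset are set at spawn, while update writes the light attributes.
# The alpha-curve ScaleColor is disabled, and no shown module reads User.GlowColor or User.GlowScale.
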NiagaraSystem /Skyfire/Gameplay/BouncePad/Effects/NS_Skyfire_Bouncepad_Idle
dependencies {usedDataInterfaces [&NiagaraDataInterfaceCurve], usedDynamicInputs [], usedModules [&SystemState]}
emitters {LightCone_mesh {
  emitterUpdateScript [
   &EmitterState(Life Cycle Mode=::NewEnumerator1, Loop Behavior=::NewEnumerator1, Loop Duration Mode=::NewEnumerator1, UseLoopDelay=false)
   &SpawnBurst_Instantaneous(Spawn Count=1, Use Spawn Probability=false)
  ]
  particleSpawnScript [
   &InitializeParticle(
    Color=<rgba 0.032 0.075 0.0105 1>
    Lifetime=2
    Mesh Scale=<0.001 0.001 100>
    Mesh Scale Mode=::NewEnumerator4
    Sprite Size Mode=::NewEnumerator3
    Uniform Sprite Size=250
    UsePositionOffset=false
    Write Material Random=true
   ) {module=InitializeParticle001}
   &InitialMeshOrientation(
    Orientation Axis=<0 0 1>
    Orientation Coordinate Space=::Local
    Orientation Vector=<0 0 0>
    Use Orientation Vector=true
    Use Rotation Vector=true
   )
   &SystemLocation(MaskSpawnGroup=false, Offset=<0 0 -5>)
   /SetVariables(Particles.SpriteInitialColor:LinearColor=<rgba 0.4 1 0.1 0.3>) {isSetParameters=true}
   /SetVariables(Particles.SpriteLocation:Vector3f=&Add_Vector(A=Linked() {linkedVariable {name=Particles.Position, type=NiagaraPosition}}, B=<0 0 15>)) {isSetParameters=true}
   /SetVariables(Particles.SpriteCameraOffset=65) {isSetParameters=true}
  ]
  particleUpdateScript [
   &ParticleState(Kill Particles When Lifetime Has Elapsed=false, Let Infinitely Lived Particles Die When Emitter Deactivates=true)
   &ScaleColor(
    Scale Alpha=&FloatFromCurve(FloatCurve=DataInterface() {propertyValues {Curve {keys [
         {interpMode=RCIM_Cubic, tangentMode=RCTM_User, time=0, value=0}
         {arriveTangent=-0.0331, interpMode=RCIM_Cubic, leaveTangent=-0.0331, tangentMode=RCTM_User, time=0.5, value=1}
         {interpMode=RCIM_Cubic, tangentMode=RCTM_User, time=1.0011, value=0.002}
        ]}}})
    ScaleA=true
    ScaleRGB=true
   ) {enabled=false}
   &FN_Light_Attributes(
    Color=<rgba 0.4667 1 0.2 1>
    Create Color Attribute=true
    Create Exponent Attribute=true
    Create Position Attribute=true
    Create Radius Attribute=true
    Exponent=1
    Position=<0 0 25>
    Radius=150
    Scale Color=6
   )
  ]
  rendererClasses [
   &NiagaraMeshRendererProperties
   &NiagaraLightRendererProperties
   &NiagaraSpriteRendererProperties
  ]
  renderers [
   {rendererClass=&NiagaraMeshRendererProperties, rendererIndex=0}
   {rendererClass=&NiagaraLightRendererProperties, rendererIndex=1}
   {rendererClass=&NiagaraSpriteRendererProperties, rendererIndex=2}
  ]
 }}
summary {
 systemName=NS_Skyfire_Bouncepad_Idle
 emitters [{
   emitterName=LightCone_mesh
   rendererClasses [
    &NiagaraMeshRendererProperties
    &NiagaraLightRendererProperties
    &NiagaraSpriteRendererProperties
   ]
  }]
 userVariables [
  User.GlowColor {type=LinearColor, description="", defaultValue {struct=LinearColor, value=<rgba 0.9583 0.7374 0.4544 0.01>}}
  User.GlowScale {type=NiagaraFloat, defaultValue=13, description=""}
 ]
}
