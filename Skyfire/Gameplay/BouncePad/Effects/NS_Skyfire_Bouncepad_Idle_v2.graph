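# Niagara system for the idle effect of the Skyfire bounce pad: a persistent teal sprite glow with a green light.
# The LightCone_mesh emitter burst-spawns particles that outlive their lifetime until the emitter deactivates, and renders them as light and sprite, with no mesh renderer.
# Update modules write separate light and sprite attributes, with a curve quickly fading in sprite alpha.
# Spawn-time offsets set sprite placement and pull toward the camera.
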NiagaraSystem /Skyfire/Gameplay/BouncePad/Effects/NS_Skyfire_Bouncepad_Idle_v2
dependencies {usedDataInterfaces [&NiagaraDataInterfaceCurve], usedDynamicInputs [], usedModules [&SystemState]}
emitters {LightCone_mesh {
  emitterUpdateScript [
   &EmitterState(
    Inactive Response=::NewEnumerator1
    Life Cycle Mode=::NewEnumerator1
    Loop Behavior=::NewEnumerator1
    Loop Duration Mode=::NewEnumerator1
    UseLoopDelay=false
   )
   &SpawnBurst_Instantaneous(Use Spawn Probability=false, UseLoopCountLimit=false)
  ]
  particleSpawnScript [
   &InitializeParticle(
    Color=<rgba 1 1 1 0.5>
    Mesh Scale=<1.15 1.15 1.75>
    Mesh Scale Mode=::NewEnumerator4
    Sprite Size Mode=::NewEnumerator3
    Uniform Sprite Size=450
    UsePositionOffset=false
    Write Material Random=true
   ) {module=InitializeParticle001}
   &SystemLocation(MaskSpawnGroup=false, Offset=<0 0 -28>)
   /SetVariables(Particles.SpriteInitialColor:LinearColor=<rgba 0.9 3 2.65 0.15>) {isSetParameters=true}
   /SetVariables(Particles.SpriteLocation:Vector3f=&Add_Vector(A=Linked() {linkedVariable {name=Particles.Position, type=NiagaraPosition}}, B=<0 0 10>)) {isSetParameters=true}
   /SetVariables(Particles.SpriteCameraOffset=65) {isSetParameters=true}
   &CameraOffset(Camera Offset Amount=-15)
  ]
  particleUpdateScript [
   &ParticleState(Kill Particles When Lifetime Has Elapsed=false, Let Infinitely Lived Particles Die When Emitter Deactivates=true)
   &FN_Light_Attributes(
    Color=<rgba 0.4667 1 0.2 1>
    Create Color Attribute=true
    Create Exponent Attribute=true
    Create Position Attribute=true
    Create Radius Attribute=true
    Exponent=1
    Position=<0 0 25>
    Radius=150
    Scale Color=6
   )
   &FN_Sprite_Attributes(
    Color=&MultiplyLinearColor(
     Linear Color=$Particles.Initial.SpriteInitialColor
     Scale Factor=&MakeVector4(
      X=1
      Y=1
      Z=1
      W=&FloatFromCurve(FloatCurve=DataInterface() {propertyValues {Curve {keys [{interpMode=RCIM_Cubic, tangentMode=RCTM_User, time=0, value=0}, {interpMode=RCIM_Cubic, tangentMode=RCTM_User, time=0.15, value=1}]}}})
     )
    )
    Create Color Attribute=true
    Write to Module Namespace=true
   )
  ]
  rendererClasses [&NiagaraLightRendererProperties, &NiagaraSpriteRendererProperties]
  renderers [
   {rendererClass=&NiagaraLightRendererProperties, rendererIndex=0}
   {rendererClass=&NiagaraSpriteRendererProperties, rendererIndex=1}
  ]
 }}
summary {
 systemName=NS_Skyfire_Bouncepad_Idle_v2
 emitters [{
   emitterName=LightCone_mesh
   rendererClasses [&NiagaraLightRendererProperties, &NiagaraSpriteRendererProperties]
  }]
 userVariables []
}
